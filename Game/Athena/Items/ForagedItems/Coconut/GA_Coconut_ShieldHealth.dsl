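; Gameplay ability for the foraged Coconut that periodically restores the player's health, or shield once health is full.
; Activation caches the pawn, fires a cue if shields exist, and starts a looping scalable-interval timer, cleared on ability end.
; Each tick applies the heal or shield effect, then checks a ForceCancelAbility function that returns false here.
; Activation also sends a glitched-foraged quest stat for tagged players with an open or party-assisted objective.
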(event Custom|RecachePlayerPawn
  (bind _returnvalue (Ability|GetAvatarActorfromActorInfo))
  ;[1 1 1] Recache player pawn if necessary
  (Utilities|IsValid _returnvalue
    (:Is%20Valid
      (bind _asfort_player_pawn (Utilities|Casting|CastToFortPlayerPawn _returnvalue))
      (Variables|Default|SetPlayerPawn _asfort_player_pawn))))

(event Ability|EventFailedToActivatePassiveAbility
  (params (ActorInfo :type Gameplay%20Ability%20Actor%20Info%20Structure)))

(event Ability|EventActivateAbility
  (bind _returnvalue (Ability|GetAvatarActorfromActorInfo))
  (bind _playerpawn (Variables|Default|GetPlayerPawn))
  (bind _returnvalue_2 (Fortnite|GetFortPlayerControllerfromActor _playerpawn))
  ;[1 1 1] Apply Effect on tick
  (Utilities|IsValid _returnvalue
    (:Is%20Valid
      (bind _asfort_player_pawn (Utilities|Casting|CastToFortPlayerPawn _returnvalue))
      (Variables|Default|SetPlayerPawn _asfort_player_pawn)
      (Default|ExecuteActivateGameplayCue)
      (bind _returnvalue_3 (Utilities|Time|SetTimerbyEvent (delegate Custom|ApplyEffect) (CallFunction|GetTimeInterval) true))
      (Variables|Default|SetTimer_Interval _returnvalue_3)
      ;[1 1 1] Glitched foraged items quest
      (bind _bdidactorhavetags (Class|AthenaQuestFunctionLibrary|AthenaCheckActorHasTags _playerpawn "(GameplayTags=((TagName=\"Athena.Quests.GlitchedForaged.ConsumeGlitched\")))" true))
      (if _bdidactorhavetags
        (bind (_questisvalid _questcompleted _objectivecompleted) (Default|AthenaCheckQuestandObjectiveCompletion _returnvalue_2 0 "Quest_BR_S10_Mystery_Consume_GlitchedForaged_Different_01"))
        (bind (_bisassistingobjective _bisvalidplayercontroller) (Class|AthenaQuestFunctionLibrary|AthenaPartyAssistCheck _returnvalue_2 "Quest_BR_S10_Mystery_Consume_GlitchedForaged_Different_01"))
        (if (or (and _bisassistingobjective _bisvalidplayercontroller) (Math|Boolean|ANDBoolean _questisvalid (not _questcompleted) (not _objectivecompleted)))
          (Quest|SendComplexCustomStatEvent (Quests|GetQuestManager _returnvalue_2 "Invalid") 0 "(GameplayTags=)" "(GameplayTags=((TagName=\"Athena.Quests.GlitchedForaged.Coco\")))" 1))))))

(event Custom|ApplyEffect
  (bind _playerpawn (Variables|Default|GetPlayerPawn))
  ;[1 1 1] ApplyEffect, If Max HP do not apply
  (Utilities|IsValid _playerpawn
    (:Is%20Valid
      (if (== (FortPawn|GetMaxHealth _playerpawn) (FortPawn|GetHealth _playerpawn))
        (if (not (or (== (FortPawn|GetShield _playerpawn) (FortPawn|GetMaxShield _playerpawn)) (<= (FortPawn|GetMaxShield _playerpawn) 0)))
          (Ability|ApplyGameplayEffectToOwner (Variables|Default|GetGE_Shields) 0))
        (else
          (Ability|ApplyGameplayEffectToOwner (Variables|Default|GetGE_Heals) 0)))
      (Default|CallOnEffectApplied)
      (if (CallFunction|ForceCancelAbility)
        (Ability|CancelAbility)))))

(event Ability|EventOnEndAbility
  (params (bWasCancelled :type Boolean))
  ;[1 1 1] Clear timer on end
  (Utilities|Time|ClearandInvalidateTimerbyHandle (Variables|Default|GetTimer_Interval)))

(fn ExecuteActivateGameplayCue
  (if (not (<= (FortPawn|GetMaxShield (Variables|Default|GetPlayerPawn)) 0))
    (Ability|ExecuteGameplayCueOnOwner (Variables|Default|GetGC_Activate))))

(fn ForceCancelAbility
  (return false))

(fn GetTimeInterval
  (return (ScalableFloat|GetValueatLevel (Variables|Default|GetTimeInterval))))
